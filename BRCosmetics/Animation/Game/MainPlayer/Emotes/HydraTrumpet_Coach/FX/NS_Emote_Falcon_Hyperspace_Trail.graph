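# Niagara system for the hyperspace trail of the Falcon emote, drawing short-lived contrail streaks from a skeletal mesh.
# The only enabled emitter, ContrailsOpt, spawns ribbons on the mesh in a brief window at the loop start, tinted by a system contrail color.
# Loop-age curves taper width, ramp up launch velocity along Y and increase drag, while alpha fades and an emitter-age check kills particles.
# The AirGlow sprite and GPU BlowingParticles001 mesh emitters are disabled.
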NiagaraSystem /BRCosmetics/Animation/Game/MainPlayer/Emotes/HydraTrumpet_Coach/FX/NS_Emote_Falcon_Hyperspace_Trail
dependencies {
 usedDataInterfaces [&NiagaraDataInterfaceCurve, &NiagaraDataInterfaceColorCurve]
 usedDynamicInputs [&FNTimeOfDayColor, &Lerp_Float, &Sine]
 usedModules [
  &SystemState
  &self:SystemSpawnScript.SystemScriptSource.SystemScriptGraph.NiagaraNodeAssignment_0.SetVariables_#1
  &self:SystemSpawnScript.SystemScriptSource.SystemScriptGraph.NiagaraNodeAssignment_1.SetVariables_#2
  &self:SystemSpawnScript.SystemScriptSource.SystemScriptGraph.NiagaraNodeAssignment_2.SetVariables_#3
 ]
}
emitters {
 AirGlow {
  bEnabled=false
  emitterUpdateScript [
   &EmitterState()
   &SpawnBurst_Instantaneous(Spawn Count=1, Use Spawn Probability=false)
  ]
  particleSpawnScript [
   &InitializeParticle(
    Color=<rgba 0.4323 0.4323 0.4323 1>
    Lifetime=0.25
    Sprite Size Mode=::NewEnumerator3
    Uniform Sprite Size=120
    UsePositionOffset=false
    Write Material Random=true
   ) {module=InitializeParticle001}
   &DynamicMaterialParameters(
    Index 0 Param 1=2
    Module.Param0WriteEnabled=true
    Module.Param1WriteEnabled=true
    Module.Param2WriteEnabled=true
    Module.Param3WriteEnabled=true
   )
   &CameraOffset(Camera Offset Amount=10)
  ]
  particleUpdateScript [
   &ParticleState()
   &ScaleSpriteSize(Scale Factor=&Vector2DFromFloat(Value=&FloatFromCurve(FloatCurve=DataInterface() {propertyValues {Curve {keys [<at 0 1 RCIM_Cubic>]}}})))
   &ScaleColor(
    Scale Alpha=&FloatFromCurve(FloatCurve=DataInterface() {propertyValues {Curve {keys [<at 0 0>,<at 0.25 0.75>,<at 1 0>]}}})
    ScaleA=true
    ScaleRGB=true
   ) {module=ScaleColor001}
  ]
  rendererClasses [&NiagaraSpriteRendererProperties]
  renderers [{rendererClass=&NiagaraSpriteRendererProperties}]
 }
 BlowingParticles001 {
  bEnabled=false
  simTarget=GPUComputeSim
  emitterUpdateScript [
   &EmitterState(Life Cycle Mode=::NewEnumerator1, UseLoopDelay=false)
   &SpawnBurst_Instantaneous(Spawn Count=2, Use Spawn Probability=false, UseLoopCountLimit=false)
   /SetVariables(Emitter.float:NiagaraFloat=&FloatFromCurve(FloatCurve=DataInterface() {propertyValues {Curve {keys [<at 7.75 40>,<at 8.3 1>]}}}, CurveIndex=$Emitter.Age)) {isSetParameters=true}
  ]
  particleSpawnScript [
   &PartitionParticles()
   &InitializeParticle(
    Lifetime Max=15
    Lifetime Min=12
    Lifetime Mode=::NewEnumerator1
    Mass Max=1.5
    Mass Mode=::NewEnumerator3
    Mesh Scale Mode=::NewEnumerator3
    Mesh Uniform Scale=&MaskFloat(Mask Source=0, Mask Parameter=?, Mask=$Particles.PartitionParticles.Partition, Masked Float=1, Passthrough Float=-1)
    Position Offset=&MaskVector(Mask Source=0, Mask Parameter=?, Mask=$Particles.PartitionParticles.Partition, Masked Vector=<0 -7 1.6>, Passthrough Vector=<0 -50 1.6>)
    Sprite Rotation Mode=::NewEnumerator2
    Sprite Size Mode=::NewEnumerator1
    Uniform Sprite Size Max=8
    Uniform Sprite Size Min=4
    UsePositionOffset=true
    Write Material Random=true
   )
   &DynamicMaterialParameters(
    Index 0 Param 1=1
    Index 0 Param 2=0.3
    Module.Index0Float0Write=true
    Module.Index0Float1Write=true
    Module.Index0Float2Write=true
    Module.Index0Float3Write=true
   )
  ]
  particleUpdateScript [
   &ParticleState()
   /SetVariables(
    Particles.GlowCameraOffset=1
    Particles.GlowColor=&Multiply_LinearColorByFloat(
     Linear Color=&MakeLinearColorFromVectorAndFloat("Vector (RGB)"=<0.16 0.32 1>, "Float (Alpha)"=0.1)
     Float=&FloatFromCurve(FloatCurve=DataInterface() {propertyValues {Curve {keys [<at 7.75 5>,<at 8.3 10>]}}}, CurveIndex=$Emitter.Age)
    )
   ) {isSetParameters=true}
   &ScaleColor(
    Color Value To Scale=$Particles.GlowColor
    Scale Alpha=&FloatFromCurve(FloatCurve=DataInterface() {propertyValues {Curve {keys [<at 0 0>,<at 0.1 1>,<at 1 1>]}}})
    ScaleA=true
    ScaleRGB=true
   )
  ]
  rendererClasses [&NiagaraMeshRendererProperties]
  renderers [{rendererClass=&NiagaraMeshRendererProperties, rendererIndex=0}]
 }
 ContrailsOpt {
  emitterUpdateScript [
   &EmitterState(Life Cycle Mode=::NewEnumerator1, Loop Behavior=::NewEnumerator1, Loop Delay=0.07, Loop Duration=3, UseLoopDelay=false)
   &SpawnRate(SpawnRate=60, Use Spawn Probability=false) {enabled=false}
   &SpawnRate(
    SpawnRate=&FloatFromCurve(FloatCurve=DataInterface() {propertyValues {Curve {keys [<at 0.09 1>,<at 0.1 0>]}}}, CurveIndex=$Emitter.NormalizedLoopAge, Scale Curve=256)
    Use Spawn Probability=false
   ) {module=SpawnRate001}
  ]
  particleSpawnScript [
   &FN_SetSpawnGroup(Number of Spawn Groups=1)
   &InitializeRibbon(
    Color=&MultiplyLinearColor(
     Linear Color=$System.Contrails
     Scale Factor=&MakeVector4(
      X=1
      Y=1
      Z=1
      W=&FloatFromCurve(FloatCurve=DataInterface() {propertyValues {Curve {keys [<at 0 0>,<at 0.05 1>,<at 0.1 0>]}}}, CurveIndex=$Emitter.NormalizedLoopAge)
     )
    )
    Ribbon Facing Mode:ENiagara_UnsetDirectSet=::NewEnumerator1
    Ribbon Facing Vector:Vector3f=<1 0 0>
    Ribbon Width=&FloatFromCurve(FloatCurve=DataInterface() {propertyValues {Curve {keys [<at 0 1>,<at 0.06 0>]}}}, CurveIndex=$Emitter.NormalizedLoopAge, Scale Curve=16)
    UsePositionOffset=false
    Write Material Random=true
   )
   &SkeletalMeshLocation(Bone Sampling Mode=::NewEnumerator0, MaskSpawnGroup=false, OverrideRandomSeed=false, Skeletal Mesh=$System.Skeletal Mesh)
   &DynamicMaterialParameters(
    Index 0 Param 2=&RandomRangeFloat(Minimum=0, Maximum=1)
    Module.Index0Float0Write=false
    Module.Index0Float1Write=true
    Module.Index0Float2Write=false
    Module.Index0Float3Write=false
   ) {enabled=false}
   &AddVelocity(Velocity=&Lerp_Vector(
     A=<0 0 0>
     B=<1 2000 1>
     Alpha=&FloatFromCurve(FloatCurve=DataInterface() {propertyValues {Curve {keys [<at 0 0>,<at 0.06 1>]}}}, CurveIndex=$Emitter.NormalizedLoopAge)
    ))
  ]
  particleUpdateScript [
   &ParticleState()
   &ScaleColor(
    Scale Alpha=&Multiply_Float(
     A=&FloatFromCurve(
      FloatCurve=DataInterface() {propertyValues {Curve {keys [
          {arriveTangent=-1, interpMode=RCIM_Cubic, time=0, value=0}
          {interpMode=RCIM_Cubic, tangentMode=RCTM_User, time=0.2, value=1}
          {interpMode=RCIM_Cubic, tangentMode=RCTM_User, time=1, value=0}
         ]}}}
      Scale Curve=&FNTimeOfDayScalar(Night Scale=0.2, Morning Scale=0.2, Evening Scale=0.2)
     )
     B=1
    )
    Scale RGB=&Multiply_VectorByFloat(
     Vector=&MakeVectorFromLinearColorRGB(Color=&ColorFromCurve(ColorCurve=DataInterface() {propertyValues {
         AlphaCurve {keys []}
         BlueCurve {keys [<at 0 4.8655 RCIM_Cubic>,<at 0.4581 4.8655>,<at 0.8818 6.0417>,<at 0.9992 2.9167 RCIM_Cubic>]}
         GreenCurve {keys [<at 0 10 RCIM_Cubic>,<at 0.4581 10>,<at 0.8818 10>,<at 0.9992 5 RCIM_Cubic>]}
         RedCurve {keys [<at 0 1.1979 RCIM_Cubic>,<at 0.4581 1.1979>,<at 0.8818 0.5>,<at 0.9992 0 RCIM_Cubic>]}
        }}))
     Float=&FloatFromCurve(
      FloatCurve=DataInterface() {propertyValues {Curve {keys [<at 0 1>,<at 0.1 1.2>,<at 0.3 1>,<at 0.5 0.5>,<at 0.7 1.5>,<at 1 1>]}}}
      Scale Curve=0.8
     )
    )
    ScaleA=true
    ScaleRGB=false
   ) {enabled=false}
   &ScaleRibbonWidth(Ribbon Width Scale=&FloatFromCurve(FloatCurve=DataInterface() {propertyValues {Curve {keys [<at 0 0.3>,<at 0.45 1>,<at 0.8 0.675>,<at 1 0.25>]}}})) {enabled=false, module=ScaleRibbonWidth001}
   &DynamicMaterialParameters(
    Index 0 Param 2=&RandomRangeFloat(Minimum=0, Maximum=1)
    Index 0 Param 3=-1
    Module.Param0WriteEnabled=false
    Module.Param1WriteEnabled=false
    Module.Param2WriteEnabled=false
    Module.Param3WriteEnabled=false
   ) {enabled=false, module=DynamicMaterialParameters001}
   &CameraOffset(Camera Offset Amount=10) {enabled=false}
   &SkeletalMeshLocation(Bone Sampling Mode=::NewEnumerator4, MaskSpawnGroup=false, Position Sampling=::NewEnumerator2, Skeletal Mesh=$User.SkeletalMesh) {enabled=false, module=SkeletalMeshLocation001}
   &PointAttractionForce(
    Attraction Strength=&FloatFromCurve(
     FloatCurve=DataInterface() {propertyValues {Curve {keys [{tangentMode=RCTM_User, time=0.3, value=0}, {tangentMode=RCTM_User, time=0.5, value=75}]}}}
     Scale Curve=&MakeCustomFloatFromBool(
      Bool=&SetBoolByFloatComparison(
       A=&Add_Float(
        A=&Add_Float(
         A=&Abs_Float(Float:NiagaraFloat=&MakeFloatFromVector(Vector=$Engine.Owner.Velocity, Channel=::NewEnumerator0))
         B=&Abs_Float(Float:NiagaraFloat=&MakeFloatFromVector(Vector=$Engine.Owner.Velocity, Channel=::NewEnumerator2))
        )
        B=&Abs_Float(Float:NiagaraFloat=&MakeFloatFromVector(Vector=$Engine.Owner.Velocity, Channel=::NewEnumerator1))
       )
       B=200
      )
      True Float=&ClampFloat(Float=1)
      False Float=&ClampFloat(Float=0)
     )
    )
    Attractor Position:NiagaraPosition=$Output.SkeletalMeshLocation001.SampledPositionWithoutOffset
    Falloff Exponent=0.45
    Kill Within Radius=false
    Module.Use Falloff=true
    UseAttractorOffset=false
   ) {enabled=false}
   &Drag(
    Drag=&FloatFromCurve(
     FloatCurve=DataInterface() {propertyValues {Curve {keys [
         {arriveTangent=118.6787, interpMode=RCIM_Cubic, leaveTangent=118.6787, tangentMode=RCTM_User, time=0, value=3}
         {arriveTangent=146.0425, interpMode=RCIM_Cubic, leaveTangent=146.0425, tangentMode=RCTM_User, time=0.12, value=20}
        ]}}}
     CurveIndex=$Emitter.NormalizedLoopAge
    )
    Use Linear Drag=true
    Use Rotational Drag=true
   )
   &SolveForcesAndVelocity(Clamp Velocity=false, Limit Acceleration=false)
   &ScaleColor(
    Scale Alpha=&FloatFromCurve(
     FloatCurve=DataInterface() {propertyValues {Curve {keys [<at 0 1>,<at 0.05 1>,<at 0.1 1>]}}}
     CurveIndex=$Emitter.NormalizedLoopAge
     Scale Curve=&FloatFromCurve(
      FloatCurve=DataInterface() {propertyValues {Curve {keys [
          {interpMode=RCIM_Cubic, tangentMode=RCTM_User, time=0, value=0}
          {interpMode=RCIM_Cubic, tangentMode=RCTM_User, time=0.1, value=1}
          {tangentMode=RCTM_User, time=1, value=0}
         ]}}}
      Scale Curve=2
     )
    )
    ScaleA=true
    ScaleRGB=true
   ) {module=ScaleColor001}
   &KillParticles(Kill Particles=&SetBoolByFloatComparison(A=$Emitter.Age, B=1))
  ]
  rendererClasses [&NiagaraRibbonRendererProperties]
  renderers [{rendererClass=&NiagaraRibbonRendererProperties}]
 }
}
summary {
 systemName=NS_Emote_Falcon_Hyperspace_Trail
 emitters [
  {emitterName=ContrailsOpt, rendererClasses [&NiagaraRibbonRendererProperties]}
  {bEnabled=false, emitterName=BlowingParticles001, simTarget=GPUComputeSim, rendererClasses [&NiagaraMeshRendererProperties]}
  {bEnabled=false, emitterName=AirGlow, rendererClasses [&NiagaraSpriteRendererProperties]}
 ]
 userVariables [
  User.NormalizedNotifyProgress {type=NiagaraFloat, defaultValue=0, description=""}
  User.SkeletalMesh {
   type=NiagaraDataInterfaceSkeletalMesh
   defaultValue {
    struct=NiagaraExt_VariableValue_DataInterface
    value {dataInterface=&self:NiagaraDataInterfaceSkeletalMesh_0}
   }
  }
 ]
}
