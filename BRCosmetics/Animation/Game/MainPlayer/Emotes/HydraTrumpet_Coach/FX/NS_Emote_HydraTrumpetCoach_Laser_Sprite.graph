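# Niagara system for the HydraTrumpet Coach emote that fires short-lived green laser bolts from a "Tie Fighter" emitter.
# Timed three-particle bursts use User.ActiveTie to switch firing times and select the sampled skeletal-mesh bone or socket.
# Partition arrays split particles by visibility tag and decide which stay pinned to that bone, with curves driving drag, stretch and fade.
# Despite the Sprite name, all three renderers are mesh renderers.
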NiagaraSystem /BRCosmetics/Animation/Game/MainPlayer/Emotes/HydraTrumpet_Coach/FX/NS_Emote_HydraTrumpetCoach_Laser_Sprite
dependencies {
 usedDataInterfaces [
  &NiagaraDataInterfaceArrayInt32
  &NiagaraDataInterfaceArrayFloat3
  /Script/Niagara.NiagaraDataInterfaceArrayBool
  &NiagaraDataInterfaceCurve
  &NiagaraDataInterfaceVector2DCurve
  &NiagaraDataInterfaceColorCurve
  &NiagaraDataInterfaceSkeletalMesh
 ]
 usedDynamicInputs [&RandomBool]
 usedModules [
  &self:SystemSpawnScript.SystemScriptSource.SystemScriptGraph.NiagaraNodeAssignment_1.SetVariables_#1
  &SystemState
  &self:SystemSpawnScript.SystemScriptSource.SystemScriptGraph.NiagaraNodeAssignment_0.SetVariables_#2
 ]
}
emitters {Tie_Fighter_1 {
  emitterUpdateScript [
   &EmitterState(
    Enable Distance Culling=true
    Enable Visibility Culling=true
    Inactive Response=::NewEnumerator1
    Life Cycle Mode=::NewEnumerator1
    Loop Behavior=::NewEnumerator1
    Loop Duration Mode=::NewEnumerator1
    MaxDistance=10000
    Module.bUseMaxDistance=true
    Module.bUseMinDistance=false
    Scalability Mode=::NewEnumerator1
    UseLoopDelay=false
   )
   &SpawnBurst_Instantaneous(
    Loop Count Limit=12
    Spawn Count=3
    Spawn Time=&RandomRangeFloat(Minimum=0, Maximum=0.05)
    Use Spawn Probability=false
    UseLoopCountLimit=true
   ) {enabled=false, module=SpawnBurst_Instantaneous003}
   &SpawnBurst_Instantaneous(
    Loop Count Limit=12
    Spawn Count=3
    Spawn Probability=0.3
    Spawn Time=&RandomRangeFloat(Minimum=0.17, Maximum=0.2)
    Use Spawn Probability=true
    UseLoopCountLimit=true
   ) {enabled=false, module=SpawnBurst_Instantaneous004}
   &SpawnBurst_Instantaneous(
    Loop Count Limit=12
    Spawn Count=3
    Spawn Probability=&Lerp_Float(A=1, B=0, Alpha=$User.ActiveTie)
    Spawn Time=0.32
    Use Spawn Probability=true
    UseLoopCountLimit=false
   )
   &SpawnBurst_Instantaneous(
    Loop Count Limit=12
    Spawn Count=3
    Spawn Probability=&Lerp_Float(A=0, B=1, Alpha=$User.ActiveTie)
    Spawn Time=0.48
    Use Spawn Probability=true
    UseLoopCountLimit=false
   ) {module=SpawnBurst_Instantaneous011}
   &SpawnBurst_Instantaneous(
    Loop Count Limit=12
    Spawn Count=3
    Spawn Probability=&Lerp_Float(A=1, B=0, Alpha=$User.ActiveTie)
    Spawn Time=1.79
    Use Spawn Probability=true
    UseLoopCountLimit=false
   ) {module=SpawnBurst_Instantaneous001}
   &SpawnBurst_Instantaneous(
    Loop Count Limit=12
    Spawn Count=3
    Spawn Probability=&Lerp_Float(A=1, B=0, Alpha=$User.ActiveTie)
    Spawn Time=1.84
    Use Spawn Probability=true
    UseLoopCountLimit=false
   ) {module=SpawnBurst_Instantaneous009}
   &SpawnBurst_Instantaneous(
    Loop Count Limit=12
    Spawn Count=3
    Spawn Probability=&Lerp_Float(A=0, B=1, Alpha=$User.ActiveTie)
    Spawn Time=1.9
    Use Spawn Probability=true
    UseLoopCountLimit=false
   ) {enabled=false, module=SpawnBurst_Instantaneous012}
   &SpawnBurst_Instantaneous(
    Loop Count Limit=12
    Spawn Count=3
    Spawn Probability=&Lerp_Float(A=1, B=0, Alpha=$User.ActiveTie)
    Spawn Time=2.45
    Use Spawn Probability=true
    UseLoopCountLimit=false
   ) {enabled=false, module=SpawnBurst_Instantaneous002}
   &SpawnBurst_Instantaneous(
    Loop Count Limit=12
    Spawn Count=3
    Spawn Probability=&Lerp_Float(A=1, B=0, Alpha=$User.ActiveTie)
    Spawn Time=2.48
    Use Spawn Probability=true
    UseLoopCountLimit=false
   ) {enabled=false, module=SpawnBurst_Instantaneous010}
   &SpawnBurst_Instantaneous(
    Loop Count Limit=12
    Spawn Count=3
    Spawn Probability=&Lerp_Float(A=1, B=0, Alpha=$User.ActiveTie)
    Spawn Time=4.17
    Use Spawn Probability=true
    UseLoopCountLimit=false
   ) {enabled=false, module=SpawnBurst_Instantaneous005}
   &SpawnBurst_Instantaneous(
    Loop Count Limit=12
    Spawn Count=3
    Spawn Probability=&Lerp_Float(A=0, B=1, Alpha=$User.ActiveTie)
    Spawn Time=4.8
    Use Spawn Probability=true
    UseLoopCountLimit=false
   ) {module=SpawnBurst_Instantaneous013}
   &SpawnBurst_Instantaneous(
    Loop Count Limit=12
    Spawn Count=3
    Spawn Probability=&Lerp_Float(A=1, B=0, Alpha=$User.ActiveTie)
    Spawn Time=5.19
    Use Spawn Probability=true
    UseLoopCountLimit=false
   ) {module=SpawnBurst_Instantaneous007}
   &SpawnBurst_Instantaneous(
    Loop Count Limit=12
    Spawn Count=3
    Spawn Probability=&Lerp_Float(A=1, B=0, Alpha=$User.ActiveTie)
    Spawn Time=5.48
    Use Spawn Probability=true
    UseLoopCountLimit=false
   ) {module=SpawnBurst_Instantaneous008}
   /SetVariables(Emitter.Vector:Vector3f=&RandomRangeVector(Minimum=<0.005 400 0.005>, Maximum=<-0.005 500 -0.005>)) {isSetParameters=true}
   &SpawnBurst_Instantaneous(Spawn Count=3, Spawn Time=1.65, Use Spawn Probability=false, UseLoopCountLimit=false) {module=SpawnBurst_Instantaneous006}
  ]
  particleSpawnScript [
   &PartitionParticles(Number Of Partitions=4)
   /SetVariables(Particles.VisibilityTag=&SelectIntFromArray(Direct Array Index=$Particles.PartitionParticles.Partition, Int Selection Array=DataInterface() {propertyValues {IntData [0,0,1,1]}})) {isSetParameters=true}
   &InitializeParticle(
    Color=<rgba 1.6562 50 0 1>
    Lifetime=0.3
    Mesh Scale Mode=::NewEnumerator3
    Mesh Uniform Scale=0.02
    Position Offset=&SelectVectorFromArray(
     Direct Array Index=$Particles.PartitionParticles.Partition
     Vector Selection Array=DataInterface() {propertyValues {FloatData [<-3 0 0>, <3 0 0>, <-3 0 0>, <3 0 0>]}}
    )
    Sprite Rotation Angle=180
    Sprite Rotation Mode=::NewEnumerator0
    Sprite Size Mode=::NewEnumerator3
    Sprite UV Mode=::NewEnumerator4
    Uniform Sprite Size=5
    UsePositionOffset=true
    Write Material Random=true
   ) {enabled=false}
   &InitializeParticle(
    Color=<rgba 1.6562 50 0 1>
    Lifetime=0.3
    Mesh Scale Mode=::NewEnumerator3
    Mesh Uniform Scale=0.02
    Position Offset=&SelectVectorFromArray(
     Direct Array Index=$Particles.PartitionParticles.Partition
     Vector Selection Array=DataInterface() {propertyValues {FloatData [<0 0 0>, <0 0 0>, <0 0 0>, <0 0 0>]}}
    )
    Sprite Rotation Angle=180
    Sprite Rotation Mode=::NewEnumerator0
    Sprite Size Mode=::NewEnumerator3
    Sprite UV Mode=::NewEnumerator4
    Uniform Sprite Size=5
    UsePositionOffset=true
    Write Material Random=true
   ) {module=InitializeParticle001}
   &AddVelocity(Velocity=$Emitter.Vector)
   &InheritVelocity(Inherited Velocity Speed Limit=1000, Speed Limit=true, UseSpeedThreshold=false)
   &FN_Mesh_Attributes(
    Color=<rgba 0.5781 0.5781 0.5781 1>
    Create Color Attribute=true
    Create Scale Attribute=true
    Lifetime=0.1
    Scale=<0.015 0.02 0.02>
    Write to Module Namespace=true
   )
   &FN_Sprite_Attributes(Lifetime=0.3)
   &InitialMeshOrientation(Mesh Axis To Orient=<0 1 0>, Mesh Orientation Mode=::NewEnumerator2, Use Rotation Vector=false)
   &SkeletalMeshLocation(
    Bone / Socket Index=&SelectIntFromArray(
     Direct Array Index=&MakeIntFromFloat(Float=$User.ActiveTie)
     Int Selection Array=DataInterface() {propertyValues {IntData [0,1]}}
    )
    Bone Sampling Mode=::NewEnumerator4
    MaskSpawnGroup=false
    Sampled Position Offset=&SelectVectorFromArray(
     Direct Array Index=$Particles.PartitionParticles.Partition
     Vector Selection Array=DataInterface() {propertyValues {FloatData [<0 0 0>, <0 0 0>, <0 0 0>, <0 0 0>]}}
    )
    Skeletal Mesh=$System.Skeletal Mesh
    Write To Particle Position:NiagaraBool=/Niagara/DynamicInputs/Arrays/SelectBoolFromArray(
     Array Sampling Mode:ENiagara_LifetimeMode=::NewEnumerator0
     Direct Array Index:NiagaraInt32=$Particles.PartitionParticles.Partition
     Direct Array Mode:ENiagara_WrapClamp=::NewEnumerator0
     Evaluation Type:ENiagaraRandomnessEvaluation=::NewEnumerator0
     Recalculate Random Each Loop:NiagaraBool=false
     Randomness Mode:ENiagaraRandomnessMode=::NewEnumerator0
     Random Seed:NiagaraInt32=?
     Fixed Random Seed:NiagaraBool=false
     Bool Selection Array:NiagaraDataInterfaceArrayBool=DataInterface() {propertyValues {GpuSyncMode=SyncCpuToGpu, MaxElements=0, BoolData [false,false,true,true]}}
    )
   ) {module=SkeletalMeshLocation001}
  ]
  particleUpdateScript [
   &Drag(
    Drag=&FloatFromCurve(
     FloatCurve=DataInterface() {propertyValues {Curve {keys [
         {interpMode=RCIM_Cubic, tangentMode=RCTM_User, time=0, value=0}
         {arriveTangent=1.953, interpMode=RCIM_Cubic, leaveTangent=1.953, tangentMode=RCTM_User, time=0.5, value=1}
        ]}}}
     Scale Curve=30.1338
    )
    Use Linear Drag=true
    Use Rotational Drag=true
   )
   &SolveForcesAndVelocity(Clamp Velocity=false, Limit Acceleration=false)
   &ParticleState()
   &ScaleSpriteSize(
    Non-Uniform Curve Scale=<3 6>
    Non-Uniform Curve Sprite Scale=DataInterface() {propertyValues {
      XCurve {keys [
        {arriveTangent=0.2907, interpMode=RCIM_Cubic, leaveTangent=0.2907, tangentMode=RCTM_User, time=0, value=0.05}
        {arriveTangent=0.1239, interpMode=RCIM_Cubic, leaveTangent=0.1239, tangentMode=RCTM_User, time=0.3904, value=0.1}
        {interpMode=RCIM_Cubic, leaveTangent=0.6, time=0.7, value=0.1}
       ]}
      YCurve {keys [<at 0 0 RCIM_Cubic>, {arriveTangent=2.8626, interpMode=RCIM_Cubic, leaveTangent=2.8626, tangentMode=RCTM_User, time=0.3, value=1}]}
     }}
    Scale Sprite Size Mode=::NewEnumerator3
   )
   &ScaleSpriteSize(
    Scale Factor=&MakeVector2D(
     X=&FloatFromCurve(
      FloatCurve=DataInterface() {propertyValues {Curve {keys [
          {tangentMode=RCTM_User, time=0, value=1}
          {arriveTangent=-2.0083, interpMode=RCIM_Cubic, leaveTangent=-2.0083, tangentMode=RCTM_User, time=1, value=0.8}
         ]}}}
      Scale Curve=3
     )
     Y=&FloatFromCurve(
      FloatCurve=DataInterface() {propertyValues {Curve {keys [
          {tangentMode=RCTM_User, time=0, value=1}
          {arriveTangent=-2.0083, interpMode=RCIM_Cubic, leaveTangent=-2.0083, tangentMode=RCTM_User, time=1, value=1}
         ]}}}
      Scale Curve=1.3
     )
    )
    Scale Sprite Size Mode=::NewEnumerator2
   ) {module=ScaleSpriteSize001}
   &ScaleMeshSize(Scale Factor=&MakeVector(
     X=0.3
     Y=&FloatFromCurve(FloatCurve=DataInterface() {propertyValues {Curve {keys [<at 0 0>,<at 0.2 1>]}}}, Scale Curve=10)
     Z=0.5
    )) {module=ScaleMeshSize001}
   &DynamicMaterialParameters(
    Index 0 Param 1=2
    Index 0 Param 2=25
    Index 0 Param 4=1
    Module.Param0WriteEnabled=true
    Module.Param1WriteEnabled=true
    Module.Param2WriteEnabled=false
    Module.Param3WriteEnabled=false
   )
   &ScaleColor(
    Scale Alpha=&Lerp_Float(
     A=&FloatFromCurve(FloatCurve=DataInterface() {propertyValues {Curve {keys [
          {arriveTangent=-2.5249, interpMode=RCIM_Cubic, leaveTangent=-2.5249, tangentMode=RCTM_User, time=0, value=0.9847}
          {arriveTangent=-0.0405, interpMode=RCIM_Cubic, leaveTangent=-0.0405, tangentMode=RCTM_User, time=0.4, value=0}
         ]}}})
     B=&FloatFromCurve(FloatCurve=DataInterface() {propertyValues {Curve {keys [
          {arriveTangent=-2.5249, interpMode=RCIM_Cubic, leaveTangent=-2.5249, tangentMode=RCTM_User, time=0, value=0.9847}
          {arriveTangent=-0.0405, interpMode=RCIM_Cubic, leaveTangent=-0.0405, tangentMode=RCTM_User, time=0.6, value=0}
         ]}}})
     Alpha=&MakeFloatFromInt(Int=$Particles.Initial.VisibilityTag)
    )
    ScaleA=true
    ScaleRGB=false
   )
   /SetVariables(
    Particles.CenterColor:LinearColor=&Multiply_LinearColorByFloat(
     Linear Color=<rgba 45.4497 49.353 50 1>
     Float=&FloatFromCurve(FloatCurve=DataInterface() {propertyValues {Curve {keys [<at 0 1>,<at 0.5 0>]}}})
    )
    Particles.CenterOffset=1
    Particles.CenterSize:Vector2f=&Multiply_Vector2D(A=$Particles.SpriteSize, B=<0.3 0.6>)
    Particles.CenterSizeVector:Vector3f=&Multiply_Vector(A=$Particles.Scale, B=<1 1 1>)
   ) {isSetParameters=true}
   &SkeletalMeshLocation(
    Bone / Socket Index=&SelectIntFromArray(
     Direct Array Index=&MakeIntFromFloat(Float=$User.ActiveTie)
     Int Selection Array=DataInterface() {propertyValues {IntData [0,1]}}
    )
    Bone Sampling Mode=::NewEnumerator4
    MaskSpawnGroup=false
    Orientation Sampling=::NewEnumerator1
    Sampled Position Offset=&SelectVectorFromArray(
     Direct Array Index=$Particles.PartitionParticles.Partition
     Vector Selection Array=DataInterface() {propertyValues {FloatData [<0 0 0>, <0 0 0>, <0 0 0>, <0 0 0>]}}
    )
    Skeletal Mesh=$System.Skeletal Mesh
    Write To Particle Position:NiagaraBool=&SelectBoolFromArray(
     Array Sampling Mode:ENiagara_LifetimeMode=::NewEnumerator0
     Direct Array Index:NiagaraInt32=$Particles.PartitionParticles.Partition
     Direct Array Mode:ENiagara_WrapClamp=::NewEnumerator0
     Evaluation Type:ENiagaraRandomnessEvaluation=::NewEnumerator0
     Recalculate Random Each Loop:NiagaraBool=false
     Randomness Mode:ENiagaraRandomnessMode=::NewEnumerator0
     Random Seed:NiagaraInt32=?
     Fixed Random Seed:NiagaraBool=false
     Bool Selection Array:NiagaraDataInterfaceArrayBool=DataInterface() {propertyValues {GpuSyncMode=SyncCpuToGpu, MaxElements=0, BoolData [false,false,true,true]}}
    )
   )
   /SetVariables(Particles.MuzzleOrientation:Quat4f=$Particles.SkeletalMeshLocation.SampledRotation) {isSetParameters=true}
   /SetVariables(Particles.Linear Color=&ColorFromCurve(ColorCurve=DataInterface() {propertyValues {
       AlphaCurve {keys [<at 0 1 RCIM_Cubic>,<at 0.3937 0 RCIM_Cubic>]}
       BlueCurve {keys [<at 0 0.0594 RCIM_Cubic>,<at 1 1 RCIM_Cubic>]}
       GreenCurve {keys [<at 0 0.7478 RCIM_Cubic>,<at 1 1 RCIM_Cubic>]}
       RedCurve {keys [<at 0 0.0441 RCIM_Cubic>,<at 1 1 RCIM_Cubic>]}
      }})) {isSetParameters=true}
  ]
  rendererClasses [&NiagaraMeshRendererProperties]
  renderers [
   {rendererClass=&NiagaraMeshRendererProperties, rendererIndex=0}
   {rendererClass=&NiagaraMeshRendererProperties, rendererIndex=1}
   {rendererClass=&NiagaraMeshRendererProperties, rendererIndex=2}
  ]
 }}
summary {
 systemName=NS_Emote_HydraTrumpetCoach_Laser_Sprite
 emitters [{emitterName=Tie_Fighter_1, rendererClasses [&NiagaraMeshRendererProperties]}]
 userVariables [
  User.ActiveTie {type=NiagaraFloat, defaultValue=0, description=""}
  User.Alpha {type=NiagaraFloat, defaultValue=1, description=""}
  User.bIsFrontEnd {type=NiagaraBool}
  User.bIsFrontEndPreview {type=NiagaraBool}
 ]
}
