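# Niagara system for the explosion effect in the HydraTrumpet Coach emote.
# Emitters layer sparks, flash sprites, flare and refraction-glow meshes, lights, ribbon smoke trails, smoke puffs and a ground decal.
# Bursts are timed from User.ExplosionDelay, and most spawn only when User.HitGround is true, though one spark emitter always bursts.
# Flash sizes and brightness adapt for front-end versus gameplay, and smoke trails partition one burst into several seeded ribbons.
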NiagaraSystem /BRCosmetics/Animation/Game/MainPlayer/Emotes/HydraTrumpet_Coach/FX/NS_Emote_HydraTrumpet_Explosion
dependencies {
 usedDataInterfaces [
  &NiagaraDataInterfaceVector2DCurve
  &NiagaraDataInterfaceCurve
  &NiagaraDataInterfaceVectorCurve
  &NiagaraDataInterfaceColorCurve
 ]
 usedDynamicInputs [
  &SimulationPosition
  &NormalizedAngleToDegrees
  &ReturnNormalizedExecIndex
  &MaskLinearColorBySpawnGroup
 ]
 usedModules [
  &self:SystemSpawnScript.SystemScriptSource.SystemScriptGraph.NiagaraNodeAssignment_3.SetVariables_#1
  &self:SystemSpawnScript.SystemScriptSource.SystemScriptGraph.NiagaraNodeAssignment_6.SetVariables_#2
  &SystemState
  &self:SystemSpawnScript.SystemScriptSource.SystemScriptGraph.NiagaraNodeAssignment_5.SetVariables_#3
 ]
}
emitters {
 Decal {
  emitterUpdateScript [
   &EmitterState(Inactive Response=::NewEnumerator2, Life Cycle Mode=::NewEnumerator1, Loop Behavior=::NewEnumerator1, UseLoopDelay=false)
   &self:Activate_Emitter(Active:NiagaraBool=$System.Activate, Lifetime=1.8) {module=Activate_Emitter}
   &SpawnBurst_Instantaneous(
    Spawn Count=&MakeCustomIntFromBool(Bool=$User.HitGround, True INT=1, False INT=0)
    Spawn Time=$User.ExplosionDelay
    Use Spawn Probability=false
    UseLoopCountLimit=false
   )
  ]
  particleSpawnScript [
   &InitializeParticle(
    Lifetime Max=1.8
    Lifetime Min=1.5
    Lifetime Mode=::NewEnumerator1
    Mass Mode=::NewEnumerator3
    Sprite Size Mode=::NewEnumerator1
    Uniform Sprite Size Max=12
    Uniform Sprite Size Min=6
    UsePositionOffset=false
    Write Material Random=true
   )
   /SetVariables(Particles.DecalOrientation:Quat4f=&MultiplyQuaternion(
     Quaternion A=<0 1 0 1>
     Quaternion B=&LerpQuaternion(
      Quaternion A:Quat4f=<30 0 0 1>
      Quaternion B:Quat4f=<-30 0 0 1>
      Lerp Factor:NiagaraFloat=&RandomRangeFloat(Minimum=-100, Maximum=100)
      Enable Spherical Linear Interpolation:NiagaraBool=false
     )
    )) {isSetParameters=true}
  ]
  particleUpdateScript [
   &ParticleState()
   &ScaleColor(ScaleA=true, ScaleRGB=true)
   /SetVariables(Particles.DecalFade:NiagaraFloat=&FloatFromCurve(FloatCurve=DataInterface() {propertyValues {Curve {keys [<at 0 0>,<at 0.1 1>,<at 0.5 0.9>,<at 1 0>]}}})) {isSetParameters=true}
   /SetVariables(Particles.DecalSize:Vector3f=&VectorFromCurve(
     VectorCurve=DataInterface() {propertyValues {XCurve {keys [<at 0 0>,<at 0.1 1>]}, YCurve {keys [<at 0 0>,<at 0.1 1>]}, ZCurve {keys [<at 0 0>,<at 0.1 1>]}}}
     Scale Curve=&Multiply_VectorByFloat(Vector=<80 80 80>, Float=&RandomRangeFloat(Minimum=0.8, Maximum=1.1))
    )) {isSetParameters=true}
  ]
  rendererClasses [&NiagaraDecalRendererProperties]
  renderers [{rendererClass=&NiagaraDecalRendererProperties, rendererIndex=0}]
 }
 Directional_sparks {
  emitterUpdateScript [
   &EmitterState(Inactive Response=::NewEnumerator2, Life Cycle Mode=::NewEnumerator1, Loop Behavior=::NewEnumerator1, UseLoopDelay=false)
   &self:Activate_Emitter(Active:NiagaraBool=$System.Activate, Lifetime=1.3) {module=Activate_Emitter}
   &SpawnBurst_Instantaneous(
    Spawn Count=&MakeCustomIntFromBool(Bool=$User.HitGround, True INT=&UniformRangedInt(Minimum=10, Maximum=15), False INT=0)
    Spawn Time=$User.ExplosionDelay
    Use Spawn Probability=false
   )
  ]
  particleSpawnScript [
   &InitializeParticle(
    Color Mode=::NewEnumerator0
    Lifetime Max=1.3
    Lifetime Min=0.5
    Lifetime Mode=::NewEnumerator1
    Sprite Size Max=<1.3333 3.3333>
    Sprite Size Min=<0.6667 1.3333>
    Sprite Size Mode=::NewEnumerator2
    UsePositionOffset=false
    Write Material Random=true
   ) {module=InitializeParticle001}
   &ShapeLocation(
    Cylinder Height=10
    Cylinder Radius=2
    MaskSpawnGroup=false
    Override Seed=false
    Shape Primitive=::NewEnumerator1
    Surface Only=false
    UseOwnerScale=true
   ) {module=ShapeLocation001}
   &CameraOffset(Camera Offset Amount=15)
   &AddVelocity(
    Module.Use Falloff Distance=false
    Override Seed=false
    Velocity Mode=::NewEnumerator1
    Velocity Speed=&RandomRangeFloat(Minimum=30, Maximum=90)
   )
   &AddVelocity(
    Cone Angle=200
    Distribution Along Cone Axis=90
    Module.UseConeFalloff=true
    Module.UseVelDistribution=true
    Override Seed=false
    Speed Falloff From Cone Axis=0.5
    Velocity Mode=::NewEnumerator2
    Velocity Speed=&RandomRangeFloat(Minimum=50, Maximum=150)
   ) {enabled=false, module=AddVelocity001}
  ]
  particleUpdateScript [
   &ParticleState()
   &Drag(Drag=&RandomRangeFloat(Minimum=1, Maximum=2), Use Linear Drag=true, Use Rotational Drag=false)
   &GravityForce(Gravity=&RandomRangeVector(Minimum=<0 0 -200>, Maximum=<0 0 -500>)) {enabled=false}
   &SolveForcesAndVelocity(Clamp Velocity=false, Limit Acceleration=false)
   &ScaleSpriteSizeBySpeed(Max Scale Factor=<1 10>, Min Scale Factor=<0.65 0.65>)
   &ScaleSpriteSize(Scale Factor=&Vector2DFromCurve(Vector2Curve=DataInterface() {propertyValues {XCurve {keys [<at 0 1>,<at 1 1>]}, YCurve {keys [<at 0 1>,<at 1 0>]}}}, ScaleCurve=<1 1>))
   &ScaleColor(
    Scale RGB=&MakeVectorFromLinearColorRGB(Color=&ColorFromCurve(ColorCurve=DataInterface() {propertyValues {
        AlphaCurve {keys [<at 0 1 RCIM_Cubic>]}
        BlueCurve {keys [<at 0 0.4596 RCIM_Cubic>,<at 1 0 RCIM_Cubic>]}
        GreenCurve {keys [<at 0 2.37 RCIM_Cubic>,<at 1 0 RCIM_Cubic>]}
        RedCurve {keys [<at 0 50 RCIM_Cubic>,<at 1 1 RCIM_Cubic>]}
       }}))
    ScaleA=true
    ScaleRGB=true
   )
  ]
  rendererClasses [&NiagaraSpriteRendererProperties]
  renderers [{rendererClass=&NiagaraSpriteRendererProperties}]
 }
 Directional_sparks001 {
  emitterUpdateScript [
   &EmitterState(Inactive Response=::NewEnumerator2, Life Cycle Mode=::NewEnumerator1, Loop Behavior=::NewEnumerator1, UseLoopDelay=false)
   &self:Activate_Emitter(Active:NiagaraBool=$System.Activate, Lifetime=1.3) {module=Activate_Emitter}
   &SpawnBurst_Instantaneous(Spawn Count=100, Spawn Time=$User.ExplosionDelay, Use Spawn Probability=false)
  ]
  particleSpawnScript [
   &InitializeParticle(
    Color Mode=::NewEnumerator0
    Lifetime Max=1.3
    Lifetime Min=0.5
    Lifetime Mode=::NewEnumerator1
    Sprite Size Max=<1.3333 3.3333>
    Sprite Size Min=<0.6667 1.3333>
    Sprite Size Mode=::NewEnumerator2
    UsePositionOffset=false
    Write Material Random=true
   ) {module=InitializeParticle001}
   &ShapeLocation(
    Cylinder Height=10
    Cylinder Radius=2
    MaskSpawnGroup=false
    Override Seed=false
    Shape Primitive=::NewEnumerator1
    Surface Only=false
    UseOwnerScale=true
   ) {module=ShapeLocation001}
   &CameraOffset(Camera Offset Amount=15)
   &AddVelocity(
    Module.Use Falloff Distance=false
    Override Seed=false
    Velocity Mode=::NewEnumerator1
    Velocity Speed=&RandomRangeFloat(Minimum=10, Maximum=90)
   )
   &AddVelocity(
    Cone Angle=200
    Distribution Along Cone Axis=90
    Module.UseConeFalloff=true
    Module.UseVelDistribution=true
    Override Seed=false
    Speed Falloff From Cone Axis=0.5
    Velocity Mode=::NewEnumerator2
    Velocity Speed=&RandomRangeFloat(Minimum=50, Maximum=150)
   ) {enabled=false, module=AddVelocity001}
  ]
  particleUpdateScript [
   &ParticleState()
   &Drag(Drag=&RandomRangeFloat(Minimum=1, Maximum=2), Use Linear Drag=true, Use Rotational Drag=false)
   &GravityForce(Gravity=&RandomRangeVector(Minimum=<0 0 -200>, Maximum=<0 0 -500>)) {enabled=false}
   &SolveForcesAndVelocity(Clamp Velocity=false, Limit Acceleration=false)
   &ScaleSpriteSizeBySpeed(Max Scale Factor=<1 10>, Min Scale Factor=<0.65 0.65>)
   &ScaleSpriteSize(Scale Factor=&Vector2DFromCurve(Vector2Curve=DataInterface() {propertyValues {XCurve {keys [<at 0 1>,<at 1 1>]}, YCurve {keys [<at 0 1>,<at 1 0>]}}}, ScaleCurve=<1 1>))
   &ScaleColor(
    Scale RGB=&MakeVectorFromLinearColorRGB(Color=&ColorFromCurve(ColorCurve=DataInterface() {propertyValues {
        AlphaCurve {keys [<at 0 1 RCIM_Cubic>]}
        BlueCurve {keys [<at 0 1.1031 RCIM_Cubic>,<at 1 0 RCIM_Cubic>]}
        GreenCurve {keys [<at 0 5.688 RCIM_Cubic>,<at 1 0 RCIM_Cubic>]}
        RedCurve {keys [<at 0 120 RCIM_Cubic>,<at 1 1 RCIM_Cubic>]}
       }}))
    ScaleA=true
    ScaleRGB=true
   )
  ]
  rendererClasses [&NiagaraSpriteRendererProperties]
  renderers [{rendererClass=&NiagaraSpriteRendererProperties}]
 }
 Flash_Elements {
  emitterUpdateScript [
   &EmitterState(Inactive Response=::NewEnumerator2, Life Cycle Mode=::NewEnumerator1, Loop Behavior=::NewEnumerator1, UseLoopDelay=false)
   &self:Activate_Emitter(Active:NiagaraBool=$System.Activate, Lifetime=0.25) {module=Activate_Emitter}
   &SpawnBurst_Instantaneous(
    Spawn Count=&MakeCustomIntFromBool(Bool=$User.HitGround, True INT=1, False INT=0)
    Spawn Group=1
    Spawn Time=$User.ExplosionDelay
    Use Spawn Probability=false
    UseLoopCountLimit=false
   ) {module=SpawnBurst_Instantaneous001}
   &SpawnBurst_Instantaneous(
    Spawn Count=&MakeCustomIntFromBool(Bool=$User.HitGround, True INT=&UniformRangedInt(Minimum=6, Maximum=8), False INT=0)
    Spawn Time=$User.ExplosionDelay
    Use Spawn Probability=false
    UseLoopCountLimit=false
   )
  ]
  particleSpawnScript [
   &InitializeParticle(Color Mode=::NewEnumerator0, Lifetime=0.25, UsePositionOffset=true, Write Material Random=true)
   &ShapeLocation(MaskSpawnGroup=false, Override Seed=false, Sphere Radius=1, UseOwnerScale=true)
   &SpriteFacingAndAlignment(Sprite Alignment=$Particles.ShapeLocation.ShapeNormal, WriteAlignment=true, WriteFacing=false)
   &FN_PerParticleLOD(Enable SpriteSize Distance Scaling=true, Sprite Size Scale Far=<2.5 2.5>)
   /SetVariables(Particles.SpriteSize=<0.9 0.9>, Particles.VisibilityTag=&SpawnGroup()) {isSetParameters=true}
   /SetVariables(
    Particles.Flash Spike Lifetime:NiagaraFloat=&RandomRangeFloat(Minimum=0.07, Maximum=0.1)
    Particles.Flash Spike Sprite Size:Vector2f=&FN_FrontEndVector2Modifier(
     Front End:Vector2f=&RandomRangeVector2D(Minimum=<2.4 24>, Maximum=<3.6 36>)
     Gameplay:Vector2f=&RandomRangeVector2D(Minimum=<3 15>, Maximum=<6 24>)
    )
   ) {isSetParameters=true}
   /SetVariables(
    Particles.Flash Star Cam Offset=15
    Particles.Flash Star Lifetime=0.15
    Particles.Flash Star Rotation:NiagaraFloat=&RandomRangeFloat(Minimum=0, Maximum=360)
    Particles.Flash Star Sprite Size:Vector2f=&Multiply_Vector2D_ByFloat(
     Float=&Multiply_Float(
      A=&MakeCustomFloatFromBool(Bool=$User.bIsFrontEnd, True Float=0.45, False Float=1)
      B=&RandomRangeFloat(Minimum=36, Maximum=45)
     )
     Vector2D=$Particles.SpriteSize
    )
    Particles.Flash Star Sprite Size Alt:Vector2f=&Multiply_Vector2D_ByFloat(
     Float=&Multiply_Float(
      A=&MakeCustomFloatFromBool(Bool=$User.bIsFrontEnd, True Float=0.45, False Float=1)
      B=&RandomRangeFloat(Minimum=30, Maximum=37.5)
     )
     Vector2D=$Particles.SpriteSize
    )
    Particles.Flash Star SubUV Index:NiagaraFloat=&RandomRangeFloat(Minimum=0, Maximum=3)
    Particles.Flash Star SubUV Index Alt:NiagaraFloat=&RandomRangeFloat(Minimum=0, Maximum=3)
   ) {isSetParameters=true}
   /SetVariables(
    Particles.Flash Flare Color:LinearColor=<rgba 1 0.3834 0 0.7>
    Particles.Flash Flare Lifetime:NiagaraFloat=&FN_FrontEndFloatModifier(Gameplay=0.2, Front End=0.125)
    Particles.Flash Flare Sprite Size:Vector2f=&FN_FrontEndVector2Modifier(Front End:Vector2f=<27 27>, Gameplay:Vector2f=<90 90>)
   ) {isSetParameters=true}
   &DynamicMaterialParameters(
    Index 0 Param 1=0.2
    Index 0 Param 2=1
    Module.Index0Param0Write=true
    Module.Index0Param1Write=true
    Module.Index0Param2Write=true
    Module.Index0Param3Write=true
   )
   &CameraOffset(Camera Offset Amount=20)
  ]
  particleUpdateScript [
   &ParticleState()
   /SetVariables(
    Particles.Update Flash Flare Age:NiagaraFloat=&Divide_Float(A=$Particles.Age, B=$Particles.Initial.Flash Flare Lifetime)
    Particles.Update Flash Spike Age:NiagaraFloat=&Divide_Float(A=$Particles.Age, B=$Particles.Initial.Flash Spike Lifetime)
    Particles.Update Flash Star Age:NiagaraFloat=&Divide_Float(A=$Particles.Age, B=$Particles.Initial.Flash Star Lifetime)
   ) {isSetParameters=true}
   /SetVariables(
    Particles.Update Flash Spike Color:LinearColor=&MakeLinearColorFromVectorAndFloat(
     "Vector (RGB)"=&MakeVectorFromLinearColorRGB(Color=$Particles.Initial.Flash Spike Color)
     "Float (Alpha)"=&FloatFromCurve(FloatCurve=DataInterface() {propertyValues {Curve {keys [<at 0.7 1>,<at 1 0>]}}}, CurveIndex=$Particles.Update Flash Spike Age)
    )
    Particles.Update Flash Spike Sprite Size:Vector2f=&Multiply_Vector2D_ByFloat(
     Float=&FloatFromCurve(FloatCurve=DataInterface() {propertyValues {Curve {keys [<at 0 0.6>,<at 1 2>]}}}, CurveIndex=$Particles.Update Flash Spike Age)
     Vector2D=$Particles.Initial.Flash Spike Sprite Size
    )
    Particles.Update Flash Star Color:LinearColor=&ColorFromCurve(
     ColorCurve=DataInterface() {propertyValues {
       AlphaCurve {keys [<at 0 1>,<at 0.8992 0>]}
       BlueCurve {keys [<at 0.1978 0>,<at 0.4068 0.1188>,<at 0.6127 0.031>]}
       GreenCurve {keys [<at 0.1978 0.8224>,<at 0.4068 0.4324>,<at 0.6127 0.1045>]}
       RedCurve {keys [<at 0.1978 3>,<at 0.4068 2>,<at 0.6127 1.5>]}
      }}
     CurveIndex=$Particles.Update Flash Star Age
    )
    Particles.Update Flash Star Sprize Size:Vector2f=&Multiply_Vector2D_ByFloat(
     Float=&FloatFromCurve(
      FloatCurve=DataInterface() {propertyValues {Curve {keys [
          {arriveTangent=2.7013, interpMode=RCIM_Cubic, leaveTangent=2.7013, tangentMode=RCTM_User, time=0, value=0}
          {interpMode=RCIM_Cubic, tangentMode=RCTM_User, time=1, value=1}
         ]}}}
      CurveIndex=$Particles.Update Flash Star Age
     )
     Vector2D=$Particles.Initial.Flash Star Sprite Size
    )
    Particles.Update Flash Star Sprize Size Alt:Vector2f=&Multiply_Vector2D_ByFloat(
     Float=&FloatFromCurve(
      FloatCurve=DataInterface() {propertyValues {Curve {keys [
          {arriveTangent=2.7013, interpMode=RCIM_Cubic, leaveTangent=2.7013, tangentMode=RCTM_User, time=0, value=0}
          {interpMode=RCIM_Cubic, tangentMode=RCTM_User, time=1, value=1}
         ]}}}
      CurveIndex=$Particles.Update Flash Star Age
     )
     Vector2D=$Particles.Initial.Flash Star Sprite Size Alt
    )
   ) {isSetParameters=true}
   /SetVariables(Particles.Update Flash Flare Color:LinearColor=&MakeLinearColorFromVectorAndFloat(
     "Vector (RGB)"=&Multiply_Vector(
      A=&MakeVectorFromLinearColorRGB(Color=$Particles.Initial.Flash Flare Color)
      B=&VectorFromFloat(Value=&FN_FrontEndFloatModifier(Gameplay=1.5, Front End=1))
     )
     "Float (Alpha)"=&Multiply_Float(
      A=&MakeFloatFromLinearColor(LinearColor=<rgba 1 1 1 1>, Channel:ENiagaraLinearColor_Channels=::NewEnumerator3)
      B=&Lerp_Float(A=1, B=0, Alpha=$Particles.Update Flash Flare Age)
     )
    )) {isSetParameters=true}
   &FN_Light_Attributes(
    Color=<rgba 10 3.6259 0 1>
    Create Color Attribute=true
    Create Exponent Attribute=true
    Create Position Attribute=true
    Create Radius Attribute=true
    Exponent=1
    Position=&AddVectorToPosition(
     Position=/Niagara/DynamicInputs/Helpers/SimulationPosition()
     Vector=&TransformVector(Vector=<13.0797 0 16.6165>, Destination Space=::World)
    )
    Radius=90
    Scale Color=&FloatFromCurve(FloatCurve=DataInterface() {propertyValues {Curve {keys [
         {interpMode=RCIM_Cubic, tangentMode=RCTM_User, time=0, value=1}
         {arriveTangent=-0.5426, interpMode=RCIM_Cubic, leaveTangent=-0.5426, tangentMode=RCTM_User, time=1, value=0}
        ]}}})
   )
  ]
  rendererClasses [&NiagaraSpriteRendererProperties, &NiagaraLightRendererProperties]
  renderers [
   {rendererClass=&NiagaraSpriteRendererProperties}
   {rendererClass=&NiagaraSpriteRendererProperties, rendererIndex=1}
   {rendererClass=&NiagaraSpriteRendererProperties, rendererIndex=2}
   {rendererClass=&NiagaraSpriteRendererProperties, rendererIndex=3}
   {rendererClass=&NiagaraLightRendererProperties, rendererIndex=4}
  ]
 }
 LWE_Directional_Sparks {bEnabled=false, rendererClasses [], renderers []}
 LWE_Refract {bEnabled=false, rendererClasses [], renderers []}
 NE_Flare {
  emitterUpdateScript [
   &EmitterState() {module=EmitterState001}
   &self:Activate_Emitter(Active:NiagaraBool=$System.Activate, Lifetime=0.4) {enabled=false, module=Activate_Emitter}
   &SpawnBurst_Instantaneous(
    Spawn Count=&MakeCustomIntFromBool(Bool=$User.HitGround, True INT=1, False INT=0)
    Spawn Time=$User.ExplosionDelay
    Use Spawn Probability=false
   )
   &SpawnBurst_Instantaneous(
    Spawn Count=&MakeCustomIntFromBool(Bool=$User.HitGround, True INT=1, False INT=0)
    Spawn Group=1
    Spawn Time=$User.ExplosionDelay
    Use Spawn Probability=false
   ) {module=SpawnBurst_Instantaneous001}
  ]
  particleSpawnScript [
   &InitializeParticle(
    Color=<rgba 5 1.0919 0 1>
    Lifetime=&MaskFloatBySpawnGroup(
     Masked Float=0.25
     Passthrough Float=&MaskFloatBySpawnGroup(Masked Float=0.4, Passthrough Float=0.15, Spawn Group Mask=1)
    )
    Mass Mode=::NewEnumerator1
    Mesh Scale=&VectorFromFloat(Value=&MaskFloatBySpawnGroup(Masked Float=0.7, Passthrough Float=0.15))
    Mesh Scale Mode=::NewEnumerator4
    Position Offset=$System.SystemOffset
    Sprite Size=<8 8>
    Sprite Size Mode=::NewEnumerator4
    UsePositionOffset=true
    Write Material Random=true
   ) {module=InitializeParticle001}
   /SetVariables(Particles.VisibilityTag=&ModuloInt(A=&SpawnGroup(), B=2)) {isSetParameters=true}
   &InitialMeshOrientation(Orientation Axis=<0 0 1>, Orientation Vector=<0 0 1>, Rotation=<0 0.25 0>, Use Orientation Vector=true, Use Rotation Vector=true)
   &InitialMeshOrientation(
    Orientation Axis=<0 0 0>
    Orientation Vector=<0 0 0>
    Rotation=&RandomRangeVector(Minimum=<0 0 0>, Maximum=<1 0 0>)
    Use Orientation Vector=false
    Use Rotation Vector=true
   ) {module=InitialMeshOrientation001}
   /SetVariables(Particles.BigFlareOrient:Quat4f=&OrientMeshToVector(Look At Direction=<0 0 0>, Reference Axis Vector=<1 0 0>)) {isSetParameters=true}
   /SetVariables(Particles.BigFlareDynmat:Vector4f=&MakeVector4(X=&UniformRangedFloat(Minimum=0, Maximum=20), Y=0, Z=0, W=0)) {isSetParameters=true}
  ]
  particleUpdateScript [
   &ParticleState()
   &ScaleMeshSize(Scale Factor=&VectorFromFloat(Value=&MaskFloatBySpawnGroup(
      Masked Float=&FloatFromCurve(FloatCurve=DataInterface() {propertyValues {Curve {keys [{arriveTangent=5.2157, interpMode=RCIM_Cubic, leaveTangent=5.2157, tangentMode=RCTM_User, time=0, value=0}, <at 0.4 1 RCIM_Cubic>]}}})
      Passthrough Float=&FloatFromCurve(FloatCurve=DataInterface() {propertyValues {Curve {keys [
           {interpMode=RCIM_Cubic, tangentMode=RCTM_Break, time=0, value=0.2}
           {arriveTangent=2.9365, interpMode=RCIM_Cubic, leaveTangent=2.9365, tangentMode=RCTM_User, time=0.1, value=1}
           {interpMode=RCIM_Cubic, tangentMode=RCTM_Break, time=1, value=2}
          ]}}})
     )))
   &ScaleColor(
    Scale Alpha=&MaskFloatBySpawnGroup(
     Masked Float=0
     Passthrough Float=&FloatFromCurve(FloatCurve=DataInterface() {propertyValues {Curve {keys [
          {interpMode=RCIM_Cubic, tangentMode=RCTM_User, time=0, value=1}
          {arriveTangent=-2.6886, interpMode=RCIM_Cubic, leaveTangent=-2.6886, tangentMode=RCTM_User, time=1, value=0}
         ]}}})
    )
    Scale RGB=&MaskVectorBySpawnGroup(
     Masked Vector=&VectorFromFloat(Value=&FloatFromCurve(FloatCurve=DataInterface() {propertyValues {Curve {keys [<at 0 0.7>,<at 0.7729 0.7>,<at 1 0>]}}}, Scale Curve=0.2))
     Passthrough Vector=&Multiply_Vector(
      A=&MakeVectorFromLinearColorRGB(Color=&ColorFromCurve(ColorCurve=DataInterface() {propertyValues {
          AlphaCurve {keys [<at 0 1>,<at 0.4696 1>,<at 1 0>]}
          BlueCurve {keys [<at 0 0>,<at 0.2681 0>,<at 0.4696 0>,<at 1 0>]}
          GreenCurve {keys [<at 0 0.5481>,<at 0.2681 0.3605>,<at 0.4696 0.1134>,<at 1 0.049>]}
          RedCurve {keys [<at 0 1>,<at 0.2681 1>,<at 0.4696 0.6979>,<at 1 0.3698>]}
         }}))
      B=&VectorFromFloat(Value=&FloatFromCurve(FloatCurve=DataInterface() {propertyValues {Curve {keys [
            {interpMode=RCIM_Cubic, tangentMode=RCTM_Break, time=0, value=10}
            {arriveTangent=-6.6072, interpMode=RCIM_Cubic, leaveTangent=-6.6072, tangentMode=RCTM_User, time=0.15, value=3}
            {interpMode=RCIM_Cubic, tangentMode=RCTM_Break, time=1, value=1}
           ]}}}))
     )
    )
    ScaleA=true
    ScaleRGB=true
   )
   &DynamicMaterialParameters(
    Index 0 Param 1=&FloatFromCurve(FloatCurve=DataInterface() {propertyValues {Curve {keys [{arriveTangent=-2.1015, interpMode=RCIM_Cubic, leaveTangent=-2.1015, tangentMode=RCTM_User, time=0, value=0.5}, <at 1 -0.15>]}}})
    Index 0 Param 2=&FloatFromCurve(FloatCurve=DataInterface() {propertyValues {Curve {keys [{arriveTangent=-0.3464, interpMode=RCIM_Cubic, leaveTangent=-0.3464, tangentMode=RCTM_User, time=0, value=1}, <at 1 0.9>]}}})
    Index 0 Param 3=&FloatFromCurve(FloatCurve=DataInterface() {propertyValues {Curve {keys [
         {interpMode=RCIM_Cubic, tangentMode=RCTM_Break, time=0, value=0}
         {arriveTangent=0.4478, interpMode=RCIM_Cubic, leaveTangent=0.4478, tangentMode=RCTM_User, time=0.15, value=1}
         {interpMode=RCIM_Cubic, tangentMode=RCTM_Break, time=1, value=1.2}
        ]}}})
    Module.Param0WriteEnabled=true
    Module.Param1WriteEnabled=true
    Module.Param2WriteEnabled=true
    Module.Param3WriteEnabled=false
   )
   /SetVariables(Particles.BigFlareScale:Vector3f=&Multiply_VectorByFloat(Vector=$Particles.Scale, Float=0.4)) {isSetParameters=true}
   /SetVariables(Particles.BigFlareColor:LinearColor=&MultiplyLinearColor(
     Linear Color=<rgba 12 3.1502 0 1>
     Scale Factor=&MakeFloat4FromVectorAndFloat(
      "Vector (RGB)"=&VectorFromFloat(Value=1)
      "Float (Alpha)"=&FloatFromCurve(FloatCurve=DataInterface() {propertyValues {Curve {keys [<at 0.4 1 RCIM_Cubic>,<at 1 0>]}}})
     )
    )) {isSetParameters=true}
   &FN_Light_Attributes(
    Color=$Particles.Color
    Create Color Attribute=true
    Create Exponent Attribute=true
    Create Position Attribute=true
    Create Radius Attribute=true
    Exponent=3
    Position=&AddVectorToPosition(Position=$Engine.Owner.Position, Vector=$System.SystemOffset)
    Radius=450
    Scale Color=1500
   )
   &SolveForcesAndVelocity(Clamp Velocity=false, Limit Acceleration=false)
  ]
  rendererClasses [&NiagaraMeshRendererProperties]
  renderers [
   {rendererClass=&NiagaraMeshRendererProperties, rendererIndex=0}
   {rendererClass=&NiagaraMeshRendererProperties, rendererIndex=1}
  ]
 }
 NE_RefracGlowSphere {
  emitterUpdateScript [
   &EmitterState()
   &self:Activate_Emitter(Active:NiagaraBool=$System.Activate, Lifetime=0.15) {enabled=false, module=Activate_Emitter}
   &SpawnBurst_Instantaneous(
    Spawn Count=&MakeCustomIntFromBool(Bool=$User.HitGround, True INT=1, False INT=0)
    Spawn Time=$User.ExplosionDelay
    Use Spawn Probability=false
   )
  ]
  particleSpawnScript [&InitializeParticle(
    Color=<rgba 5 1.3829 0 2>
    Lifetime=0.15
    Mesh Scale Mode=::NewEnumerator3
    Mesh Uniform Scale=&MakeFloatFromVector(
     Vector=&VectorFromFloat(Value=&RandomRangeFloat(Minimum=0.25, Maximum=0.35))
     Channel=::NewEnumerator0
    )
    Position Offset=$System.SystemOffset
    UsePositionOffset=true
    Write Material Random=true
   ) {module=InitializeParticle001}]
  particleUpdateScript [
   &ParticleState()
   &DynamicMaterialParameters(
    Index 0 Param 1=&FloatFromCurve(FloatCurve=DataInterface() {propertyValues {Curve {keys [<at 0 1>,<at 0.3172 0.85>,<at 1 1>]}}})
    Index 0 Param 2=&FloatFromCurve(FloatCurve=DataInterface() {propertyValues {Curve {keys [<at 0 15>,<at 1 0>]}}})
    Module.Param0WriteEnabled=true
    Module.Param1WriteEnabled=true
    Module.Param2WriteEnabled=true
    Module.Param3WriteEnabled=true
   )
   &ScaleMeshSize(Scale Factor=/Niagara/DynamicInputs/ScaleByCurve/ScaleVectorByCurve(
     InitialValue:Vector3f=<1 1 1>
     ScaleCurve:NiagaraDataInterfaceVectorCurve=DataInterface() {propertyValues {
       XCurve {keys [
         {arriveTangent=4.2859, interpMode=RCIM_Cubic, leaveTangent=4.2859, tangentMode=RCTM_User, time=0, value=0}
         {arriveTangent=2, interpMode=RCIM_Cubic, leaveTangent=2, tangentMode=RCTM_User, time=0.25, value=1}
         {interpMode=RCIM_Cubic, tangentMode=RCTM_User, time=1, value=2}
        ]}
       YCurve {keys [
         {arriveTangent=5.8758, interpMode=RCIM_Cubic, leaveTangent=5.8759, tangentMode=RCTM_User, time=0, value=0}
         {arriveTangent=2, interpMode=RCIM_Cubic, leaveTangent=2, tangentMode=RCTM_User, time=0.25, value=1}
         {interpMode=RCIM_Cubic, tangentMode=RCTM_User, time=1, value=2}
        ]}
       ZCurve {keys [
         {arriveTangent=4.7007, interpMode=RCIM_Cubic, leaveTangent=4.7007, tangentMode=RCTM_User, time=0, value=0}
         {arriveTangent=2, interpMode=RCIM_Cubic, leaveTangent=2, tangentMode=RCTM_User, time=0.25, value=1}
         {interpMode=RCIM_Cubic, tangentMode=RCTM_User, time=1, value=2}
        ]}
      }}
     CurveIndex:NiagaraFloat=$Particles.NormalizedAge
    ))
   &ScaleColor(
    Scale Alpha=&FloatFromCurve(FloatCurve=DataInterface() {propertyValues {Curve {keys [
         {interpMode=RCIM_Cubic, tangentMode=RCTM_User, time=0.0045, value=0.9397}
         {arriveTangent=-1.0036, interpMode=RCIM_Cubic, leaveTangent=-1.0036, tangentMode=RCTM_User, time=0.4647, value=0.5233}
         {arriveTangent=-2.4196, interpMode=RCIM_Cubic, leaveTangent=-2.4196, tangentMode=RCTM_User, time=1.0011, value=0.002}
        ]}}})
    Scale RGB=&VectorFromCurve(VectorCurve=DataInterface() {propertyValues {XCurve {keys [<at -0.0011 1>,<at 0.5 0>]}, YCurve {keys [<at -0.0011 1>,<at 0.5 0>]}, ZCurve {keys [<at -0.0011 1>,<at 0.5 0>]}}})
    ScaleA=true
    ScaleRGB=true
   )
  ]
  rendererClasses [&NiagaraMeshRendererProperties]
  renderers [{rendererClass=&NiagaraMeshRendererProperties, rendererIndex=0}]
 }
 NE_SmokeTrails {
  emitterSpawnScript [
   /SetVariables(Emitter.RibbonNumber=4) {isSetParameters=true}
   /SetVariables(Emitter.ParticlePerRibbon=10) {isSetParameters=true}
  ]
  emitterUpdateScript [
   &EmitterState(Inactive Response=::NewEnumerator2, Life Cycle Mode=::NewEnumerator1, Loop Behavior=::NewEnumerator1, UseLoopDelay=false)
   &self:Activate_Emitter(Active:NiagaraBool=$System.Activate, Lifetime=0.4) {module=Activate_Emitter}
   &SpawnBurst_Instantaneous(
    Spawn Count=&MakeCustomIntFromBool(
     Bool=$User.HitGround
     True INT=&Multiply_Int(A=$Emitter.ParticlePerRibbon, B=$Emitter.RibbonNumber)
     False INT=0
    )
    Spawn Time=&Add_Float(A=$User.ExplosionDelay, B=0.1)
    Use Spawn Probability=false
    UseLoopCountLimit=false
   )
  ]
  particleSpawnScript [
   &PartitionParticles(
    ID to Partition:NiagaraInt32=&ReturnExecIndex()
    Normalized Index In Partition Includes One=true
    Number Of Partitions=$Emitter.RibbonNumber
    Particles Per Partition:NiagaraInt32=$Emitter.ParticlePerRibbon
    Partition Mode:ENiagara_PartitionMode=::NewEnumerator1
    Write Partition Index to Ribbon Link Order:ENiagara_AttributeSamplingApplyOutput=::NewEnumerator0
    Write Partition to Ribbon ID=::NewEnumerator0
   )
   &InitializeParticle(
    Color Mode=::NewEnumerator0
    Lifetime=&RandomRangeFloat(Minimum=2, Maximum=0.7)
    Position Offset=$System.SystemOffset
    Ribbon Width=/Game/Effects/Niagara/DynamicInputs/Random/FN_SeededRandomRangeFloat(Minimum:NiagaraFloat=4, Maximum:NiagaraFloat=8, Random Seed:NiagaraInt32=$Particles.PartitionParticles.Partition)
    Ribbon Width Mode=::NewEnumerator1
    Sprite Size Mode=::NewEnumerator3
    UsePositionOffset=true
    Write Material Random=true
   )
   &ShapeLocation(
    MaskSpawnGroup=false
    Sphere Distribution=::NewEnumerator1
    Sphere Radius=10
    U Position=&Multiply_FloatByInt(
     Float=&FN_SeededRandomRangeFloat(Minimum:NiagaraFloat=0.1, Maximum:NiagaraFloat=0.2, Random Seed:NiagaraInt32=$Particles.PartitionParticles.Partition)
     Integer=$Particles.PartitionParticles.Partition
    )
    UseOwnerScale=true
    V Position=&FN_SeededRandomRangeFloat(Minimum:NiagaraFloat=0, Maximum:NiagaraFloat=0.65, Random Seed:NiagaraInt32=$Particles.PartitionParticles.Partition)
   )
   /SetVariables(Particles.RibbonU0Override=&OneMinusFloat(Float=$Particles.RibbonLinkOrder)) {isSetParameters=true}
  ]
  particleUpdateScript [
   &ParticleState()
   /SetVariables(Particles.Position=&Lerp_Vector(
     A=$Particles.ShapeLocation.ShapePosition
     B=&AddVectorToPosition(
      Position=&AddVectorToPosition(
       Position=Linked() {linkedVariable {name=Particles.ShapeLocation.ShapePosition, type=Vector3f}}
       Vector=&Multiply_VectorByFloat(
        Vector=$Particles.ShapeLocation.ShapeNormal
        Float=&FloatFromCurve(
         FloatCurve=DataInterface() {propertyValues {Curve {keys [
             {arriveTangent=6.8705, interpMode=RCIM_Cubic, leaveTangent=6.8705, tangentMode=RCTM_User, time=0, value=0}
             {arriveTangent=1.0343, interpMode=RCIM_Cubic, leaveTangent=1.0343, tangentMode=RCTM_User, time=0.3, value=1}
             <at 1 1.5 RCIM_Cubic>
            ]}}}
         Scale Curve=&FN_SeededRandomRangeFloat(Minimum:NiagaraFloat=20, Maximum:NiagaraFloat=32, Random Seed:NiagaraInt32=$Particles.PartitionParticles.Partition)
        )
       )
      )
      Vector=<0 0 0>
     )
     Alpha=$Particles.PartitionParticles.Normalized Index In Partition
    )) {isSetParameters=true}
   &ScaleRibbonWidth(Ribbon Width Scale=&FloatFromCurve(
     FloatCurve=DataInterface() {propertyValues {Curve {keys [
         {interpMode=RCIM_Cubic, tangentMode=RCTM_User, time=0, value=1}
         {arriveTangent=-2.4825, interpMode=RCIM_Cubic, leaveTangent=-2.4825, tangentMode=RCTM_User, time=1, value=0}
        ]}}}
     CurveIndex=$Particles.PartitionParticles.Normalized Index In Partition
    ))
   &DynamicMaterialParameters(
    Index 0 Param 1=&Add_Float(
     A=&FN_SeededRandomRangeFloat(Minimum:NiagaraFloat=0, Maximum:NiagaraFloat=1, Random Seed:NiagaraInt32=$Particles.PartitionParticles.Partition)
     B=&FloatFromCurve(FloatCurve=DataInterface() {propertyValues {Curve {keys [<at 0 1>,<at 1 0>]}}}, Scale Curve=0.2)
    )
    Index 1 Param 3=&FloatFromCurve(
     FloatCurve=DataInterface() {propertyValues {Curve {keys [<at 0 0>,<at 1 2>]}}}
     CurveIndex=&Add_Float(
      A=$Particles.NormalizedAge
      B=&Multiply_Float(A=$Particles.PartitionParticles.Normalized Index In Partition, B=-0.5)
     )
    )
    Module.Index0Param0Write=true
    Module.Index0Param1Write=true
    Module.Index0Param2Write=true
    Module.Index0Param3Write=true
    Module.Index1Param0Write=true
    Module.Index1Param1Write=true
    Module.Index1Param2Write=true
    Module.Index1Param3Write=true
    Write Parameter Index 1=true
   )
   &Color(
    Color=&ColorFromCurve(ColorCurve=DataInterface() {propertyValues {
       AlphaCurve {keys [<at 0 5 RCIM_Cubic>,<at 0.5524 1>,<at 1 0 RCIM_Cubic>]}
       BlueCurve {keys [<at 0.2736 0 RCIM_Cubic>,<at 0.61 1 RCIM_Cubic>]}
       GreenCurve {keys [<at 0.2736 1.55 RCIM_Cubic>,<at 0.61 1 RCIM_Cubic>]}
       RedCurve {keys [<at 0.2736 5 RCIM_Cubic>,<at 0.61 1 RCIM_Cubic>]}
      }})
    Scale Color=&VectorFromFloat(Value=&Lerp_Float(A=0.5, B=1, Alpha=$System.CameraFadeAlpha))
   )
   &ScaleColor(
    Scale Alpha=&FloatFromCurve(FloatCurve=DataInterface() {propertyValues {Curve {keys [<at 0 1>,<at 0.2 0>]}}})
    ScaleA=true
    ScaleRGB=true
   )
  ]
  rendererClasses [&NiagaraRibbonRendererProperties]
  renderers [{rendererClass=&NiagaraRibbonRendererProperties}]
 }
 Smoke_Sphere {
  emitterUpdateScript [
   &EmitterState(Inactive Response=::NewEnumerator2, Life Cycle Mode=::NewEnumerator1, Loop Behavior=::NewEnumerator1, UseLoopDelay=false)
   &SpawnBurst_Instantaneous(Spawn Count=3, Spawn Time=0.15, Use Spawn Probability=false, UseLoopCountLimit=false) {enabled=false, module=SpawnBurst_Instantaneous001}
   &self:Activate_Emitter(Active:NiagaraBool=$System.Activate, Lifetime=1.6) {module=Activate_Emitter}
   &SpawnBurst_Instantaneous(
    Spawn Count=&MakeCustomIntFromBool(Bool=$User.HitGround, True INT=6, False INT=0)
    Spawn Time=$User.ExplosionDelay
    Use Spawn Probability=false
   )
  ]
  particleSpawnScript [
   &InitializeParticle(
    Color=&UniformRangedLinearColor(Minimum=<rgba 1 0.1477 0 1>, Maximum=<rgba 0.1198 0.084 0.0633 1>)
    Lifetime=&UniformRangedFloat(Minimum=1.2, Maximum=0.8)
    Mass Mode=::NewEnumerator1
    Position Offset=$System.SystemOffset
    Sprite Rotation Mode=::NewEnumerator2
    Sprite Size Mode=::NewEnumerator3
    Uniform Sprite Size=&RandomRangeFloat(Minimum=13.4615, Maximum=30)
    UsePositionOffset=true
    Write Material Random=true
   ) {module=InitializeParticle002}
   /SetVariables(Particles.InitSpriteRotRate=&UniformRangedFloat(Minimum=-3, Maximum=3)) {isSetParameters=true}
   &SphereLocation(
    MaskSpawnGroup=false
    Override Seed=true
    Random Seed=20
    Sphere Radius=20.5
    Surface Only=false
    UseNonUniformScale=false
    UseOffset=false
    UseSphereOrientation=false
   )
   /SetVariables(Particles.Position=&Add_Vector(A=&AddVectorToPosition(Position=$Engine.Owner.Position, Vector=$System.SystemOffset), B=<0 0 0>)) {isSetParameters=true}
   &AddVelocity(
    Velocity=$Output.SphereLocation.SphereNormal
    Velocity Speed Scale=&RandomRangeFloat(Minimum=30, Maximum=15)
   )
  ]
  particleUpdateScript [
   &KillParticles(Kill Particles=&SetBoolByFloatComparison(A=&VectorLength(Vector=$Engine.Owner.Velocity), B=70))
   &ParticleState()
   &SpriteRotationRate(Rotation Rate=&Multiply_Float(
     A=$Particles.InitSpriteRotRate
     B=&FloatFromCurve(
      FloatCurve=DataInterface() {propertyValues {Curve {keys [
          {arriveTangent=-4.0107, interpMode=RCIM_Cubic, leaveTangent=-4.0107, tangentMode=RCTM_User, time=-0.0046, value=1.0113}
          {arriveTangent=-0.2496, interpMode=RCIM_Cubic, leaveTangent=-0.2496, tangentMode=RCTM_User, time=0.5, value=0.16}
         ]}}}
      CurveIndex=$Emitter.Age
      Scale Curve=150
     )
    ))
   &ScaleColor(
    Scale Alpha=&FloatFromCurve(FloatCurve=DataInterface() {propertyValues {Curve {keys [<at 0.1219 4>,<at 0.194 1.5>,<at 0.5 1.5>,<at 1 0>]}}})
    Scale RGB=&VectorFromFloat(Value=&FloatFromCurve(FloatCurve=DataInterface() {propertyValues {Curve {keys [
          {arriveTangent=1.8145, interpMode=RCIM_Cubic, leaveTangent=4.5795, tangentMode=RCTM_Break, time=0, value=0.4}
          <at 0.3387 0.9565>
          <at 1 0.7>
         ]}}}))
    ScaleA=true
    ScaleRGB=true
   )
   /Niagara/Modules/Update/Forces/CurlNoiseForce(Noise Frequency=0.0012, Noise Strength=100, Pan Noise=true, Pan Noise Field=<0 0 1>) {enabled=false}
   &Drag(
    Drag=&FloatFromCurve(
     FloatCurve=DataInterface() {propertyValues {Curve {keys [
         {interpMode=RCIM_Cubic, leaveTangent=0.0927, tangentMode=RCTM_Break, time=0, value=1}
         {arriveTangent=-2.586, interpMode=RCIM_Cubic, leaveTangent=-2.586, tangentMode=RCTM_User, time=1, value=0}
        ]}}}
     Scale Curve=&Lerp_Float(A=3, B=4, Alpha=$Particles.Initial.MaterialRandom)
    )
    Use Linear Drag=true
    Use Rotational Drag=false
   )
   &SolveForcesAndVelocity(Clamp Velocity=false, Limit Acceleration=false)
   &ScaleSpriteSize(Scale Factor=&Vector2DFromFloat(Value=&FloatFromCurve(FloatCurve=DataInterface() {propertyValues {Curve {keys [
          {arriveTangent=8.1876, interpMode=RCIM_Cubic, leaveTangent=5.3162, tangentMode=RCTM_User, time=0, value=0.15}
          {arriveTangent=1.0025, interpMode=RCIM_Cubic, leaveTangent=1.0025, tangentMode=RCTM_User, time=0.166, value=0.7682}
          {arriveTangent=0.2754, interpMode=RCIM_Cubic, leaveTangent=0.2754, tangentMode=RCTM_User, time=1.0463, value=1}
         ]}}})))
   &DynamicMaterialParameters(
    Index 0 Param 1=&Add_Float(
     A=$Particles.RandomFloat
     B=&FloatFromCurve(
      FloatCurve=DataInterface() {propertyValues {Curve {keys [
          {arriveTangent=2.7533, interpMode=RCIM_Cubic, leaveTangent=1.8958, tangentMode=RCTM_User, time=0, value=1}
          {arriveTangent=0.8292, interpMode=RCIM_Cubic, leaveTangent=0.8292, tangentMode=RCTM_User, time=0.321, value=1.3453}
          {arriveTangent=0.0529, interpMode=RCIM_Cubic, leaveTangent=0.2021, tangentMode=RCTM_User, time=1, value=1.5}
         ]}}}
      CurveIndex=&Add_Float(A=&Multiply_Float(A=0, B=0.5), B=$Particles.NormalizedAge)
     )
    )
    Index 0 Param 2=&FloatFromCurve(FloatCurve=DataInterface() {propertyValues {Curve {keys [
         {arriveTangent=0.0453, interpMode=RCIM_Cubic, tangentMode=RCTM_Break, time=0, value=-1}
         {arriveTangent=0.0604, interpMode=RCIM_Cubic, leaveTangent=0.0604, tangentMode=RCTM_User, time=1, value=-1}
        ]}}})
    Index 0 Param 3=&FloatFromCurve(
     FloatCurve=DataInterface() {propertyValues {Curve {keys [
         {arriveTangent=-2.3871, interpMode=RCIM_Cubic, leaveTangent=-2.3871, tangentMode=RCTM_User, time=0, value=10}
         {arriveTangent=-16.6667, interpMode=RCIM_Cubic, leaveTangent=-16.6667, time=0.3, value=3}
         <at 0.6 0 RCIM_Cubic>
        ]}}}
     Scale Curve=&Lerp_Float(A=0.1, B=1, Alpha=$System.CameraFadeAlpha)
    )
    Index 0 Param 4=&FloatFromCurve(FloatCurve=DataInterface() {propertyValues {Curve {keys [<at 0.5 1>,<at 1 0>]}}})
    Index 1 Param 1=&FloatFromCurve(
     FloatCurve=DataInterface() {propertyValues {Curve {keys [
         {interpMode=RCIM_Cubic, leaveTangent=4.7088, tangentMode=RCTM_Break, time=0, value=0}
         {arriveTangent=4.2257, interpMode=RCIM_Cubic, leaveTangent=4.2257, tangentMode=RCTM_User, time=0.25, value=1}
        ]}}}
     Scale Curve=&RandomRangeFloat(Minimum=1, Maximum=1.5)
    )
    Index 1 Param 2=&FloatFromCurve(
     FloatCurve=DataInterface() {propertyValues {Curve {keys [<at 0 0.2>, {leaveTangent=0.9066, time=1, value=2}]}}}
     Scale Curve=&RandomRangeFloat(Minimum=0.1, Maximum=0.2)
    )
    Index 1 Param 3=&FloatFromCurve(FloatCurve=DataInterface() {propertyValues {Curve {keys [
         {interpMode=RCIM_Cubic, leaveTangent=6.3917, tangentMode=RCTM_Break, time=0, value=0.2}
         {arriveTangent=1.7549, interpMode=RCIM_Cubic, leaveTangent=10.6767, tangentMode=RCTM_User, time=0.25, value=1}
        ]}}})
    Index 1 Param 4=&FloatFromCurve(
     FloatCurve=DataInterface() {propertyValues {Curve {keys [{arriveTangent=-0.136, interpMode=RCIM_Cubic, leaveTangent=-0.136, tangentMode=RCTM_User, time=0, value=1}, <at 1 0.95>]}}}
     Scale Curve=&RandomRangeFloat(Minimum=1.2, Maximum=1.4)
    )
    Index 2 Param 1=&FloatFromCurve(FloatCurve=DataInterface() {propertyValues {Curve {keys [
         {interpMode=RCIM_Cubic, leaveTangent=4.2485, tangentMode=RCTM_Break, time=0, value=0}
         {arriveTangent=1.0548, interpMode=RCIM_Cubic, leaveTangent=1.0548, tangentMode=RCTM_User, time=0.17, value=0.5715}
         <at 1 1>
        ]}}})
    Index 2 Param 2=&FloatFromCurve(
     FloatCurve=DataInterface() {propertyValues {Curve {keys [
         {arriveTangent=1.9056, interpMode=RCIM_Cubic, leaveTangent=1.9056, tangentMode=RCTM_User, time=0, value=0.2}
         {arriveTangent=0.0259, interpMode=RCIM_Cubic, leaveTangent=0.0259, tangentMode=RCTM_User, time=1, value=1.4}
        ]}}}
     Scale Curve=&RandomRangeFloat(Minimum=0, Maximum=0)
    )
    Index 2 Param 3=&FloatFromCurve(FloatCurve=DataInterface() {propertyValues {Curve {keys [
         {interpMode=RCIM_Cubic, leaveTangent=2.3064, tangentMode=RCTM_Break, time=0, value=1}
         {interpMode=RCIM_Cubic, leaveTangent=11.1233, tangentMode=RCTM_Break, time=1, value=2}
        ]}}})
    Index 2 Param 4=&FloatFromCurve(
     FloatCurve=DataInterface() {propertyValues {Curve {keys [
         {arriveTangent=-0.6867, interpMode=RCIM_Cubic, leaveTangent=-0.6867, tangentMode=RCTM_User, time=0, value=1}
         {arriveTangent=-0.2589, interpMode=RCIM_Cubic, leaveTangent=-0.2589, tangentMode=RCTM_User, time=0.3, value=0.8}
         {interpMode=RCIM_Cubic, leaveTangent=-1.2956, tangentMode=RCTM_Break, time=1, value=0.75}
        ]}}}
     Scale Curve=1.6
    )
    Module.Param0WriteEnabled=true
    Module.Param0WriteEnabled1=true
    Module.Param0WriteEnabled2=true
    Module.Param1WriteEnabled=false
    Module.Param1WriteEnabled1=true
    Module.Param1WriteEnabled2=true
    Module.Param2WriteEnabled=true
    Module.Param2WriteEnabled1=true
    Module.Param2WriteEnabled2=true
    Module.Param3WriteEnabled=true
    Module.Param3WriteEnabled1=true
    Module.Param3WriteEnabled2=true
    Write Parameter Index 1=true
    Write Parameter Index 2=true
   )
   &CameraOffset(Camera Offset Amount=35)
  ]
  rendererClasses [&NiagaraSpriteRendererProperties]
  renderers [
   {rendererClass=&NiagaraSpriteRendererProperties}
   {rendererClass=&NiagaraSpriteRendererProperties, rendererIndex=1}
  ]
 }
 Streaks_Fat {
  emitterUpdateScript [
   &EmitterState(Inactive Response=::NewEnumerator2, Life Cycle Mode=::NewEnumerator1, Loop Behavior=::NewEnumerator1, UseLoopDelay=false)
   &self:Activate_Emitter(Active:NiagaraBool=$System.Activate, Lifetime=0.25) {module=Activate_Emitter}
   &SpawnBurst_Instantaneous(
    Spawn Count=&MakeCustomIntFromBool(Bool=$User.HitGround, True INT=10, False INT=0)
    Spawn Time=$User.ExplosionDelay
    Use Spawn Probability=false
   )
  ]
  particleSpawnScript [
   &InitializeParticle(
    Color=<rgba 10 1.2501 0 1>
    Lifetime=&UniformRangedFloat(Minimum=0.15, Maximum=0.25)
    Mass Mode=::NewEnumerator1
    Position Offset=$System.SystemOffset
    Sprite Size=&UniformRangedVector2D(Minimum=<0.5 5>, Maximum=<2.5 10>)
    Sprite Size Mode=::NewEnumerator4
    UsePositionOffset=true
    Write Material Random=true
   ) {module=InitializeParticle001}
   /Niagara/Modules/Spawn/Location/SphereLocation(MaskSpawnGroup=false, Module.Override Randomness=false, Sphere Radius=5, Surface Only=true)
   &AddVelocityFromPoint(
    Module.Use Falloff Distance=false
    Velocity Strength=&UniformRangedFloat(Minimum=33.3333, Maximum=66.6667)
   )
  ]
  particleUpdateScript [
   &ParticleState()
   &SolveForcesAndVelocity(Clamp Velocity=false, Limit Acceleration=false)
   &ScaleSpriteSize(Scale Factor=&Vector2DFromCurve(
     Vector2Curve=DataInterface() {propertyValues {XCurve {keys [<at 0 0.75 RCIM_Cubic>,<at 1 1 RCIM_Cubic>]}, YCurve {keys [<at 0 1 RCIM_Cubic>,<at 1 2 RCIM_Cubic>]}}}
     ScaleCurve=<1 1>
    ))
   &ScaleColor(
    Scale Alpha=&FloatFromCurve(FloatCurve=DataInterface() {propertyValues {Curve {keys [<at 0 1 RCIM_Cubic>,<at 1 0>]}}})
    Scale RGB=&VectorFromFloat(Value=&FloatFromCurve(
      FloatCurve=DataInterface() {propertyValues {Curve {keys [<at 0 10 RCIM_Cubic>,<at 1 1 RCIM_Cubic>]}}}
      Scale Curve=&Lerp_Float(A=0.05, B=1, Alpha=$System.CameraFadeAlpha)
     ))
    ScaleA=true
    ScaleRGB=true
   )
  ]
  rendererClasses [&NiagaraSpriteRendererProperties]
  renderers [{rendererClass=&NiagaraSpriteRendererProperties}]
 }
}
summary {
 systemName=NS_Emote_HydraTrumpet_Explosion
 emitters [
  {emitterName=Streaks_Fat, rendererClasses [&NiagaraSpriteRendererProperties]}
  {emitterName=NE_RefracGlowSphere, rendererClasses [&NiagaraMeshRendererProperties]}
  {emitterName=Smoke_Sphere, rendererClasses [&NiagaraSpriteRendererProperties]}
  {emitterName=NE_Flare, rendererClasses [&NiagaraMeshRendererProperties]}
  {emitterName=NE_SmokeTrails, rendererClasses [&NiagaraRibbonRendererProperties]}
  {
   emitterName=Flash_Elements
   rendererClasses [&NiagaraSpriteRendererProperties, &NiagaraLightRendererProperties]
  }
  {emitterName=Directional_sparks, rendererClasses [&NiagaraSpriteRendererProperties]}
  {emitterName=Decal, rendererClasses [&NiagaraDecalRendererProperties]}
  {bEnabled=false, emitterName=LWE_Refract, rendererClasses []}
  {bEnabled=false, emitterName=LWE_Directional_Sparks, rendererClasses []}
  {emitterName=Directional_sparks001, rendererClasses [&NiagaraSpriteRendererProperties]}
 ]
 userVariables [
  User.bIsFrontEnd {type=NiagaraBool}
  User.bIsFrontEndPreview {type=NiagaraBool}
  User.ExplosionDelay {type=NiagaraFloat, defaultValue=0, description=""}
  User.Global Scale {type=NiagaraFloat, defaultValue=0.07, description=""}
  User.HitGround {type=NiagaraBool, defaultValue=true}
 ]
}
